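# Sound cue for the tape recorder switch-on sound used with narrative NPC audio logs in the Battle Royale voice-over content.
# A single wave player feeds the TapeRecorder_On wave straight to the output, with no randomization, modulation or mixing.
# It is routed through the Fort_VO_Narrator sound class, so it is grouped with narrator voice rather than effects.
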SoundCue /Game/Sounds/Fort_Human_VO_BR/NPCs/Shared_BR/Narrative/Audio_Logs/TapeRecorder_On_Cue
root=WavePlayer_0
nodes [{id=WavePlayer_0, properties {soundWaveAssetPtr=&/TapeRecorder_On}}]
properties {maxDistance=2097152, soundClassObject=/Game/Sounds/FortSoundClassesAndModes/Fort_VO_Narrator, volumeMultiplier=1}
